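# Sound cue for impact hits on a portapotty object, routed through the Fort_Impacts sound class.
# A Random node picks one of five equally weighted portapotty impact waves, excluded from branch culling, and a Modulator varies its pitch before output.
# It uses small-object demolition hit attenuation and a shared object-impact concurrency set, alongside an override set to stop the oldest instance.
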SoundCue /Game/Sounds/Fort_Impact_Sounds/Portapotty/Impacts_Portapotty_Cue
root=Modulator_0
nodes [
 {id=Modulator_0, children [Random_1], properties {pitchMax=1.1}}
 {
  id=Random_1
  children [WavePlayer_8, WavePlayer_9, WavePlayer_10, WavePlayer_11, WavePlayer_12]
  properties {bShouldExcludeFromBranchCulling=true, weights [1,1,1,1,1]}
 }
 {id=WavePlayer_10, properties {soundWaveAssetPtr=&/IMP_PortaPotty_02}}
 {id=WavePlayer_11, properties {soundWaveAssetPtr=&/IMP_PortaPotty_03}}
 {id=WavePlayer_12, properties {soundWaveAssetPtr=&/IMP_PortaPotty_04}}
 {id=WavePlayer_8, properties {soundWaveAssetPtr=&/IMP_PortaPotty_05}}
 {id=WavePlayer_9, properties {soundWaveAssetPtr=&/IMP_PortaPotty_01}}
]
properties {
 attenuationSettings=/Game/Sounds/Attenuation/Fort_Demolition_Hit_Small_Sized_Object_Att
 maxDistance=4000
 soundClassObject=&Fort_Impacts
 volumeMultiplier=1.2
 concurrencyOverrides {maxCount=3, resolutionRule=StopOldest}
 concurrencySet [&CONC_Object_Impacts_FourMax]
}
